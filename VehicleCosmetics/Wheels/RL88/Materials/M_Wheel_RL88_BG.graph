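# Two-sided Default Lit material for the RL88 cosmetic wheel, flagged for skeletal mesh use.
# A desaturated, brightened ice diffuse texture is contrast-shaped by a power, masked by the inverted, power-adjusted red channel of a cubemap texture, then dimmed.
# An ice normal map shares the diffuse texture's UVs and ends in a reroute.
# The graph shows no links from either branch to material outputs.
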Material /VehicleCosmetics/Wheels/RL88/Materials/M_Wheel_RL88_BG
graph {
 properties {bAutomaticallySetUsageInEditor=false, bUsedWithMeshDeformer=true, bUsedWithSkeletalMesh=true, twoSided=true, usedShadingModels=Default Lit}
 nodes [
  n1 = Power(Base=TextureSample() {texture=/VehicleCosmetics/Wheels/RL88/Textures/_3Color_Cubemap_Red}.R, Exp=Constant() {r=0.2})
  n2 = TextureCoordinate()
  n3 = TextureSample(UVs=@n2) {texture=/VehicleCosmetics/Wheels/RL88/Textures/Ice_01_D}
  n4 = Constant() {r=1}
  n5 = Desaturation(None=Add(A=Divide(A=@n3.RGB, B=@n4), B=Constant() {r=0.66}), Fraction=Constant() {r=1})
  n6 = Constant() {r=2}
  Multiply(A=Constant() {r=0.4}, B=Multiply(A=OneMinus(None=@n1), B=Power(Base=@n5, Exp=@n6)))
  Reroute(None=TextureSample(UVs=@n2) {samplerType=SAMPLERTYPE_Normal, texture=/VehicleCosmetics/Wheels/RL88/Textures/Ice_01_N}.RGB)
 ]
}
