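# Translucent, two-sided, lit ice material for the RL88 vehicle cosmetic wheel, usable on skeletal meshes.
# An inverted red-channel mask, shaped by the IceOpacity exponent, multiplies both the IceColor tint and a brightened, desaturated, squared ice diffuse texture.
# The diffuse and a matching ice normal map share tiled UVs.
# The listing does not show which material inputs these results drive.
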Material /VehicleCosmetics/Wheels/RL88/Materials/M_Wheel_RL88_Ice
graph {
 properties {
  bAutomaticallySetUsageInEditor=false
  bUsedWithMeshDeformer=true
  bUsedWithSkeletalMesh=true
  blendMode=BLEND_Translucent
  translucencyLightingMode=TLM_Surface
  twoSided=true
  usedShadingModels=Default Lit
 }
 nodes [
  Constant() {r=50}
  n1 = OneMinus(None=Power(
    Base=TextureSample() {texture=/VehicleCosmetics/Wheels/RL88/Textures/_3Color_Cubemap_Red}.R
    Exp=ScalarParameter() {defaultValue=0.2, parameterName=IceOpacity}
   ))
  Multiply(A=VectorParameter() {parameterName=IceColor, defaultValue=<rgba 0.2143 0.4717 0.5208 1>}.RGB, B=@n1)
  n2 = TextureCoordinate() {uTiling=2.5, vTiling=2.5}
  n3 = Add(A=TextureSample(UVs=@n2) {texture=/VehicleCosmetics/Wheels/RL88/Textures/Ice_01_D}.RGB, B=Constant() {r=0.66})
  n4 = Constant() {r=1}
  Multiply(A=@n1, B=Power(Base=Desaturation(None=@n3, Fraction=@n4), Exp=Constant() {r=2}))
  TextureSample(UVs=@n2) {samplerType=SAMPLERTYPE_Normal, texture=/VehicleCosmetics/Wheels/RL88/Textures/Ice_01_N}
 ]
}
